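# Sound Cue that loops the CozyTea Robot U2 interior speaker music and feeds it into a source bus.
# A single looping Wave Player sends at full level to the SB_CozyTea_Robot_U2_Music_Loop source bus.
# Base submix and submix sends are disabled, so the cue outputs through the bus instead of playing directly.
# It uses the Fort ambient props sound class and a very large max distance.
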SoundCue /Hermes/Audio/CozyTea_Robot/SC_CozyTea_Robot_U2_Music_Int_SB_Loop
root=WavePlayer_0
nodes [{id=WavePlayer_0, properties {bLooping=true, soundWaveAssetPtr=&/CozyTea_Robot_U2_Music_Interior_Spk_Loop}}]
properties {
 bEnableBaseSubmix=false
 bEnableSubmixSends=false
 maxDistance=2097152
 soundClassObject=&Fort_Ambient_Props_Athena
 volumeMultiplier=1
 busSends [{audioBus=null, minSendLevel=0, sendLevel=1, soundSourceBus=&/SB_CozyTea_Robot_U2_Music_Loop}]
}
