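# Sound cue for intermittent voice lines from the CozyTea robot (U2), classed as an ambient prop sound.
# A looping node repeats a random 10 to 15 second delay, then an equally weighted random pick of three voice recordings.
# The robot's attenuation asset and a dedicated concurrency asset govern spatial falloff and simultaneous playback.
# The random node is excluded from branch culling.
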SoundCue /Hermes/Audio/CozyTea_Robot/SC_CozyTea_Robot_U2_VO
root=Looping_0
nodes [
 {id=Delay_0, children [Random_1], properties {delayMax=15, delayMin=10}}
 {id=Looping_0, children [Delay_0]}
 {id=Random_1, children [WavePlayer_3,WavePlayer_4,WavePlayer_5], properties {bShouldExcludeFromBranchCulling=true, weights [1,1,1]}}
 {id=WavePlayer_3, properties {soundWaveAssetPtr=&/CozyTea_Robot_U2_VO_Liaqn_01}}
 {id=WavePlayer_4, properties {soundWaveAssetPtr=&/CozyTea_Robot_U2_VO_Liaqn_02}}
 {id=WavePlayer_5, properties {soundWaveAssetPtr=&/CozyTea_Robot_U2_VO_Liaqn_03}}
]
properties {
 attenuationSettings=&/ATT_CozyTea_Robot
 maxDistance=1750
 soundClassObject=&Fort_Ambient_Props_Athena
 volumeMultiplier=1
 concurrencySet [&/CONC_CozyTea_Robot_U2_VO]
}
