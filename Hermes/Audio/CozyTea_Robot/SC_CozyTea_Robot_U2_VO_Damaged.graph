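# Sound Cue for the damaged voice lines of the CozyTea Robot U2.
# A root Random node picks among twelve equally weighted Liaqn wave variations and is excluded from branch culling.
# Playback uses the robot's attenuation and U2 VO concurrency settings under the Athena ambient props sound class, with virtualization disabled and raised volume and priority.
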SoundCue /Hermes/Audio/CozyTea_Robot/SC_CozyTea_Robot_U2_VO_Damaged
root=Random_0
nodes [
 {
  id=Random_0
  children [
   WavePlayer_0
   WavePlayer_1
   WavePlayer_2
   WavePlayer_6
   WavePlayer_7
   WavePlayer_8
   WavePlayer_9
   WavePlayer_10
   WavePlayer_11
   WavePlayer_12
   WavePlayer_13
   WavePlayer_14
  ]
  properties {bShouldExcludeFromBranchCulling=true, weights [1,1,1,1,1,1,1,1,1,1,1,1]}
 }
 {id=WavePlayer_0, properties {soundWaveAssetPtr=&/CozyTea_Robot_U2_VO_Damaged_Liaqn_01}}
 {id=WavePlayer_1, properties {soundWaveAssetPtr=&/CozyTea_Robot_U2_VO_Damaged_Liaqn_02}}
 {id=WavePlayer_10, properties {soundWaveAssetPtr=&/CozyTea_Robot_U2_VO_Damaged_Liaqn_08}}
 {id=WavePlayer_11, properties {soundWaveAssetPtr=&/CozyTea_Robot_U2_VO_Damaged_Liaqn_09}}
 {id=WavePlayer_12, properties {soundWaveAssetPtr=&/CozyTea_Robot_U2_VO_Damaged_Liaqn_10}}
 {id=WavePlayer_13, properties {soundWaveAssetPtr=&/CozyTea_Robot_U2_VO_Damaged_Liaqn_11}}
 {id=WavePlayer_14, properties {soundWaveAssetPtr=&/CozyTea_Robot_U2_VO_Damaged_Liaqn_12}}
 {id=WavePlayer_2, properties {soundWaveAssetPtr=&/CozyTea_Robot_U2_VO_Damaged_Liaqn_03}}
 {id=WavePlayer_6, properties {soundWaveAssetPtr=&/CozyTea_Robot_U2_VO_Damaged_Liaqn_04}}
 {id=WavePlayer_7, properties {soundWaveAssetPtr=&/CozyTea_Robot_U2_VO_Damaged_Liaqn_05}}
 {id=WavePlayer_8, properties {soundWaveAssetPtr=&/CozyTea_Robot_U2_VO_Damaged_Liaqn_06}}
 {id=WavePlayer_9, properties {soundWaveAssetPtr=&/CozyTea_Robot_U2_VO_Damaged_Liaqn_07}}
]
properties {
 attenuationSettings=&/ATT_CozyTea_Robot
 maxDistance=1750
 priority=1.25
 soundClassObject=&Fort_Ambient_Props_Athena
 virtualizationMode=Disabled
 volumeMultiplier=1.25
 concurrencySet [&/CONC_CozyTea_Robot_U2_VO]
}
